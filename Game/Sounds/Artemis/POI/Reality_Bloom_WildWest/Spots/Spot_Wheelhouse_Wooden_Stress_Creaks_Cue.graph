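# Sound cue for ambient wooden stress creaks at the Wheelhouse spot in the Wild West Reality Bloom POI, reusing wooden corridor bridge creak waves.
# A loop with a random delay between creaks feeds a modulator that varies pitch and volume.
# A quality switch picks from all nine creaks on its first two levels and a four-creak subset on the other two.
# It uses big-interior creak attenuation and the Fort_Doors sound class.
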SoundCue /Game/Sounds/Artemis/POI/Reality_Bloom_WildWest/Spots/Spot_Wheelhouse_Wooden_Stress_Creaks_Cue
root=Modulator_0
nodes [
 {id=Delay_0, children [QualityLevel_0], properties {delayMax=25, delayMin=3}}
 {id=Looping_0, children [Delay_0]}
 {id=Modulator_0, children [Looping_0], properties {pitchMin=0.8, volumeMin=0.8}}
 {id=QualityLevel_0, children [Random_0,Random_0,Random_1,Random_1]}
 {
  id=Random_0
  children [WavePlayer_0,WavePlayer_1,WavePlayer_2,WavePlayer_3,WavePlayer_4,WavePlayer_5,WavePlayer_6,WavePlayer_7,WavePlayer_8]
  properties {weights [1,1,1,1,1,1,1,1,1]}
 }
 {id=Random_1, children [WavePlayer_9, WavePlayer_10, WavePlayer_11, WavePlayer_12], properties {weights [1,1,1,1]}}
 {
  id=WavePlayer_0
  properties {soundWaveAssetPtr=/Game/Sounds/Artemis/POI/Normal_Bridges/Bridges_Wooden_Corridor/Spots/Spot_Bridge_Wooden_Corridor_Creaks_09}
 }
 {
  id=WavePlayer_1
  properties {soundWaveAssetPtr=/Game/Sounds/Artemis/POI/Normal_Bridges/Bridges_Wooden_Corridor/Spots/Spot_Bridge_Wooden_Corridor_Creaks_01}
 }
 {
  id=WavePlayer_10
  properties {soundWaveAssetPtr=&Spot_Bridge_Wooden_Corridor_Creaks_01}
 }
 {
  id=WavePlayer_11
  properties {soundWaveAssetPtr=/Game/Sounds/Artemis/POI/Normal_Bridges/Bridges_Wooden_Corridor/Spots/Spot_Bridge_Wooden_Corridor_Creaks_02}
 }
 {
  id=WavePlayer_12
  properties {soundWaveAssetPtr=/Game/Sounds/Artemis/POI/Normal_Bridges/Bridges_Wooden_Corridor/Spots/Spot_Bridge_Wooden_Corridor_Creaks_03}
 }
 {
  id=WavePlayer_2
  properties {soundWaveAssetPtr=&Spot_Bridge_Wooden_Corridor_Creaks_02}
 }
 {
  id=WavePlayer_3
  properties {soundWaveAssetPtr=&Spot_Bridge_Wooden_Corridor_Creaks_03}
 }
 {
  id=WavePlayer_4
  properties {soundWaveAssetPtr=/Game/Sounds/Artemis/POI/Normal_Bridges/Bridges_Wooden_Corridor/Spots/Spot_Bridge_Wooden_Corridor_Creaks_04}
 }
 {
  id=WavePlayer_5
  properties {soundWaveAssetPtr=/Game/Sounds/Artemis/POI/Normal_Bridges/Bridges_Wooden_Corridor/Spots/Spot_Bridge_Wooden_Corridor_Creaks_05}
 }
 {
  id=WavePlayer_6
  properties {soundWaveAssetPtr=/Game/Sounds/Artemis/POI/Normal_Bridges/Bridges_Wooden_Corridor/Spots/Spot_Bridge_Wooden_Corridor_Creaks_06}
 }
 {
  id=WavePlayer_7
  properties {soundWaveAssetPtr=/Game/Sounds/Artemis/POI/Normal_Bridges/Bridges_Wooden_Corridor/Spots/Spot_Bridge_Wooden_Corridor_Creaks_07}
 }
 {
  id=WavePlayer_8
  properties {soundWaveAssetPtr=/Game/Sounds/Artemis/POI/Normal_Bridges/Bridges_Wooden_Corridor/Spots/Spot_Bridge_Wooden_Corridor_Creaks_08}
 }
 {
  id=WavePlayer_9
  properties {soundWaveAssetPtr=&Spot_Bridge_Wooden_Corridor_Creaks_09}
 }
]
properties {
 attenuationSettings=/Game/Sounds/Attenuation/Fort_TileAmbient_OneShot_Big_Interior_Creak_Att
 maxDistance=2641.8145
 soundClassObject=&Fort_Doors
 volumeMultiplier=0.5
}
